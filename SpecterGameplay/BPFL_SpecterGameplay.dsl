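; Blueprint function library for Specter gameplay with a helper that builds Gameplay Cue parameters from a Fortnite gameplay ability.
; MakeOwnerInstigatorCueParams takes the effect context from the ability's owner and sets the ability's avatar actor as instigator.
; Magnitudes, location and normal are zeroed, effect and ability levels are fixed at one, and location replication via minimal-replication proxy is off.
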(fn MakeOwnerInstigatorCueParams
  (params (Ability :type Fort%20Gameplay%20Ability%20Object%20Reference%20%28by%20ref%29))
  (return (Ability|GameplayCue|MakeGameplayCueParameters :NormalizedMagnitude 0.0 :RawMagnitude 0.0 :EffectContext (Ability|GetContextfromOwner :self Ability) :Location "0,0,0" :Normal "0,0,0" :Instigator (Ability|GetAvatarActorfromActorInfo Ability) :GameplayEffectLevel 1 :AbilityLevel 1 :bReplicateLocationWhenUsingMinimalRepProxy false)))
